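; Transient quest ability for the hunter side of a Fortnite bounty: hunt a target player before time expires.
; On server activation it adds a shrinking radius-tracker circle on the target, sets a one-minute warning and binds target death, hunter death and expiry.
; A dedicated server starts the urgent quest UI.
; Squad kills complete the bounty; other kills can count as poaching; outcomes notify spectators and clean up, with a delayed fallback for ending the quest.
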(event Ability|EventActivateAbility
  (bind _returnvalue (Ability|GetAvatarActorfromActorInfo))
  (bind _returnvalue_2 (Fortnite|GetFortPlayerControllerfromActor _returnvalue))
  ;[0 0 0] Activate
  (if (Networking|IsServer)
    (bind _bountytarget (CallFunction|GetBountyTarget))
    (Utilities|IsValid _bountytarget
      (:Is%20Valid
        (Utilities|Time|SetTimerbyEvent (EventDispatchers|CreateEvent :!event Custom|OneMinuteLeft) (- (Math|Float|SelectFloat (Fortnite|GetNumericCVar "Fort.Bounty.TotalQuestTime") (ScalableFloat|GetValueatLevel (Variables|Default|GetTotalQuestTime)) (and (> (Fortnite|GetNumericCVar "Fort.Bounty.TotalQuestTime") 0.0) (Editor|InEditororPIE))) 59.0))
        (EventDispatchers|AssignOnPawnDiedDelegate _bountytarget (EventDispatchers|CreateEvent :!event Custom|OnTargetDied_Bind))
        ;;[1 0 0] NOTE: We can't bind to pawn died and expect it to get hit because the ability gets ended before delegates can multicast.
        (EventDispatchers|BindEventtoOnDied (Game|Player|GetPlayerPawn _returnvalue_2) (EventDispatchers|CreateEvent :!event Custom|OnHunterDied))
        (EventDispatchers|AssignOnUrgentQuestTimeExpired (Class|FortPlayerControllerAthena|GetTransientQuestsComponent (Utilities|Casting|CastToFortPlayerControllerAthena _returnvalue_2)) (EventDispatchers|CreateEvent :!event Custom|OnUrgentQuestTimeExpired_Event))
        (CallFunction|AddTrackingCircleInfo (ScalableFloat|GetValueatLevel (Variables|Default|GetMaxCenterOffset)) (ScalableFloat|GetValueatLevel (Variables|Default|GetMinimumRadius)) (ScalableFloat|GetValueatLevel (Variables|Default|GetMaximumRadius)) (ScalableFloat|GetValueatLevel (Variables|Default|GetTimeBetweenUpdates)) (ScalableFloat|GetValueatLevel (Variables|Default|GetRadiusShrinkAmountPerUpdate)) (Variables|Default|GetCircleColor))
        (Utilities|FlowControl|DelayUntilNextFrame)
        (Development|PrintString (Utilities|GetDisplayName (CallFunction|GetTrackedHunterBountyTarget)) true true "(R=0.917840,G=0.087147,B=1.000000,A=1.000000)" 2e+02)
        (Development|PrintString (Utilities|GetDisplayName (CallFunction|GetTrackedPrimaryHunter)) true true "(R=0.917840,G=0.087147,B=1.000000,A=1.000000)" 2e+02)
        (CallFunction|PlaceholderPostAbilityInitialization)))))

(event Custom|Placeholder_PostAbilityInitialization
  ;[0 0 0] OnPostAbilityInitialization
  ;;[1 1 1] HasAuthority returns true for Clients, and we only want this on the Server
  (if (Networking|IsDedicatedServer)
    (CallFunction|QuestStarted)
    (bind _asfort_urgent_quest_item_definition (Utilities|Casting|CastToFortUrgentQuestItemDefinition (Variables|FortTransientQuestGameplayAbility|GetQuestItemDefinition)))
    (bind _bountytarget (CallFunction|GetBountyTarget))
    (Variables|Default|SetQuestItemDefEventTag (Utilities|Struct|BreakUrgentQuestData (Class|FortUrgentQuestItemDefinition|GetUrgentQuestData _asfort_urgent_quest_item_definition)))
    (bind _structout (Utilities|Struct|SetmembersinUrgentQuestData (Class|FortUrgentQuestItemDefinition|GetUrgentQuestData _asfort_urgent_quest_item_definition) (Math|Float|SelectFloat (Fortnite|GetNumericCVar "Fort.Bounty.TotalQuestTime") (ScalableFloat|GetValueatLevel (Variables|Default|GetTotalQuestTime)) (and (> (Fortnite|GetNumericCVar "Fort.Bounty.TotalQuestTime") 0.0) (Editor|InEditororPIE))) (Utilities|Text|FormatText (.EventSubtitle (Utilities|Struct|BreakUrgentQuestData (Class|FortUrgentQuestItemDefinition|GetUrgentQuestData _asfort_urgent_quest_item_definition))) (Class|FortPlayerControllerGameplay|GetSettingsFilteredPlayerName (Utilities|Casting|CastToFortPlayerControllerGameplay (Fortnite|GetFortPlayerControllerfromActor (Ability|GetAvatarActorfromActorInfo))) _bountytarget)) (select (CallFunction|GetIsPrimaryHunter) "" (Utilities|Text|FormatText (.EventSubtitleSecondary (Utilities|Struct|BreakUrgentQuestData (Class|FortUrgentQuestItemDefinition|GetUrgentQuestData _asfort_urgent_quest_item_definition))) (Class|FortPlayerControllerGameplay|GetSettingsFilteredPlayerName (Utilities|Casting|CastToFortPlayerControllerGameplay (Fortnite|GetFortPlayerControllerfromActor (Ability|GetAvatarActorfromActorInfo))) (CallFunction|GetTrackedPrimaryHunter)))) (Utilities|Text|FormatText (.EventSubtitle (Utilities|Struct|BreakUrgentQuestData (Class|FortUrgentQuestItemDefinition|GetUrgentQuestData _asfort_urgent_quest_item_definition))) (Class|FortPlayerControllerGameplay|GetSettingsFilteredPlayerName (Utilities|Casting|CastToFortPlayerControllerGameplay (Fortnite|GetFortPlayerControllerfromActor (Ability|GetAvatarActorfromActorInfo))) _bountytarget)) _bountytarget (CallFunction|GetTrackedPrimaryHunter) (Class|FortPlayerStateZone|GetSocialAvatarBrush _bountytarget true)))
    (CallFunction|StartUrgentQuestEvent _structout)
    (CallFunction|SpectatorShowUrgentQuestMessage _structout)))

(event Custom|OnUrgentQuestTimeExpired_Event
  (params (UrgentQuestData :type Urgent%20Quest%20Data%20Structure%20%28by%20ref%29))
  (bind _urgentquestdata (Utilities|Struct|MakeUrgentQuestData :EventTag "(TagName=\"SpecialEvent.HuntingPlayer.IsHunter\")" :VisualStyleTag "(GameplayTags=)" :TotalEventTime 6e+01 :bDisplayTimer true :bQuestCompleteOnTimeExpiration false :bQuickIntro false :bQuickOutro false :bRewardPersist false :bDisplayMultiStepExpanded true :EventTitle "NSLOCTEXT(\"[38730286452D14E44F4B9F92BEE31616]\", \"AD9B85034C44D7F301B74F90AE09F1E6\", \"Bounty Failed\")" :EventSubtitle "NSLOCTEXT(\"[38730286452D14E44F4B9F92BEE31616]\", \"E18C190C4C71B46E3FE37C8BCEEE69B3\", \"Time Expired\")" :EventDescription "NSLOCTEXT(\"[38730286452D14E44F4B9F92BEE31616]\", \"F49E27F7481563D45E45F395AF2DF98F\", \"Description\")" :EventDescriptionOverrideTag "(TagName=\"\")" :SortPriority 0 :bHideQuestProgressBar false :bDisplayAllObjectivesIndividually true :QuestProgressTextFormat "LOCGEN_FORMAT_ORDERED(INVTEXT(\"`{{0}} / `{{1}}\"), INVTEXT(\"CurrentProgress\"), INVTEXT(\"MaxProgress\"))" :bReverseProgressBar true :bShowBountyThreatInformation false :bShowBountyPriceAndDistance false :bDisplayStartedEventUI true :bIsSharedUrgentQuest false :SpecialEventStartedSound "/Game/Sounds/Fort_UI_Sounds/Gameplay/Bounty/Bounty_FailBounty_Cue.Bounty_FailBounty_Cue" :AlertIconBorderColor "(R=1.000000,G=0.456411,B=0.000000,A=1.000000)" :AlertIconBrushSize "(X=0.000000,Y=0.000000)" :bOverrideWidgetPrimaryColor false :OverriddenWidgetPrimaryColor "(R=1.000000,G=0.456411,B=0.000000,A=1.000000)" :ObjectiveCategory "Default"))
  ;[0 0 0] OnUrgentQuestTimeExpired_Event
  (bind _asfort_urgent_quest_item_definition (Utilities|Casting|CastToFortUrgentQuestItemDefinition (Variables|FortTransientQuestGameplayAbility|GetQuestItemDefinition)))
  (if (GameplayTags|Equal(GameplayTag) (Utilities|Struct|BreakUrgentQuestData (Class|FortUrgentQuestItemDefinition|GetUrgentQuestData _asfort_urgent_quest_item_definition)) (Utilities|Struct|BreakUrgentQuestData UrgentQuestData))
    (CallFunction|DisplayMessageFailure _urgentquestdata)
    (CallFunction|SpectatorFailUrgentQuest _urgentquestdata (Variables|Default|GetQuestItemDefEventTag))
    (CallFunction|Cleanup)))

(event EventOnPostAbilityInitialization
  ;[0 0 0] OnPostAbilityInitialization
  ;;[1 1 1] HasAuthority returns true for Clients, and we only want this on the Server
  (if (Networking|IsDedicatedServer)
    (CallFunction|QuestStarted)
    (bind _asfort_urgent_quest_item_definition (Utilities|Casting|CastToFortUrgentQuestItemDefinition (Variables|FortTransientQuestGameplayAbility|GetQuestItemDefinition)))
    (bind _bountytarget (CallFunction|GetBountyTarget))
    (Variables|Default|SetQuestItemDefEventTag (Utilities|Struct|BreakUrgentQuestData (Class|FortUrgentQuestItemDefinition|GetUrgentQuestData _asfort_urgent_quest_item_definition)))
    (bind _structout (Utilities|Struct|SetmembersinUrgentQuestData (Class|FortUrgentQuestItemDefinition|GetUrgentQuestData _asfort_urgent_quest_item_definition) (Math|Float|SelectFloat (Fortnite|GetNumericCVar "Fort.Bounty.TotalQuestTime") (ScalableFloat|GetValueatLevel (Variables|Default|GetTotalQuestTime)) (and (> (Fortnite|GetNumericCVar "Fort.Bounty.TotalQuestTime") 0.0) (Editor|InEditororPIE))) (Utilities|Text|FormatText (.EventSubtitle (Utilities|Struct|BreakUrgentQuestData (Class|FortUrgentQuestItemDefinition|GetUrgentQuestData _asfort_urgent_quest_item_definition))) (Class|FortPlayerControllerGameplay|GetSettingsFilteredPlayerName (Utilities|Casting|CastToFortPlayerControllerGameplay (Fortnite|GetFortPlayerControllerfromActor (Ability|GetAvatarActorfromActorInfo))) _bountytarget)) (select (CallFunction|GetIsPrimaryHunter) "" (Utilities|Text|FormatText (.EventSubtitleSecondary (Utilities|Struct|BreakUrgentQuestData (Class|FortUrgentQuestItemDefinition|GetUrgentQuestData _asfort_urgent_quest_item_definition))) (Class|FortPlayerControllerGameplay|GetSettingsFilteredPlayerName (Utilities|Casting|CastToFortPlayerControllerGameplay (Fortnite|GetFortPlayerControllerfromActor (Ability|GetAvatarActorfromActorInfo))) (CallFunction|GetTrackedPrimaryHunter)))) (Utilities|Text|FormatText (.EventSubtitle (Utilities|Struct|BreakUrgentQuestData (Class|FortUrgentQuestItemDefinition|GetUrgentQuestData _asfort_urgent_quest_item_definition))) (Class|FortPlayerControllerGameplay|GetSettingsFilteredPlayerName (Utilities|Casting|CastToFortPlayerControllerGameplay (Fortnite|GetFortPlayerControllerfromActor (Ability|GetAvatarActorfromActorInfo))) _bountytarget)) _bountytarget (CallFunction|GetTrackedPrimaryHunter) (Class|FortPlayerStateZone|GetSocialAvatarBrush _bountytarget true)))
    (CallFunction|StartUrgentQuestEvent _structout)
    (CallFunction|SpectatorShowUrgentQuestMessage _structout)))

(event Custom|QuestCompleted
  (params (Title :type Text) (Subtitle :type Text) (Special%20Event%20Started%20Sound :type Sound%20Base%20Soft%20Object%20Reference))
  (bind _urgentquestdata (Utilities|Struct|MakeUrgentQuestData :EventTag "(TagName=\"SpecialEvent.HuntingPlayer.IsHunter\")" :VisualStyleTag "(GameplayTags=)" :TotalEventTime 6e+01 :bDisplayTimer true :bQuestCompleteOnTimeExpiration false :bQuickIntro false :bQuickOutro false :bRewardPersist false :bDisplayMultiStepExpanded true :EventTitle Title :EventSubtitle Subtitle :EventDescriptionOverrideTag "(TagName=\"\")" :SortPriority 0 :bHideQuestProgressBar false :bDisplayAllObjectivesIndividually true :QuestProgressTextFormat "LOCGEN_FORMAT_ORDERED(INVTEXT(\"`{{0}} / `{{1}}\"), INVTEXT(\"CurrentProgress\"), INVTEXT(\"MaxProgress\"))" :bReverseProgressBar true :bShowBountyThreatInformation false :bShowBountyPriceAndDistance false :bDisplayStartedEventUI true :bIsSharedUrgentQuest false :SpecialEventStartedSound Special%20Event%20Started%20Sound :AlertIconBorderColor "(R=1.000000,G=0.456411,B=0.000000,A=1.000000)" :AlertIconBrushSize "(X=0.000000,Y=0.000000)" :bOverrideWidgetPrimaryColor false :OverriddenWidgetPrimaryColor "(R=1.000000,G=0.456411,B=0.000000,A=1.000000)" :ObjectiveCategory "Default"))
  ;[0 0 0] Server_QuestCompleted
  (Variables|Default|SetIsQuestCompleted true)
  (CallFunction|DisplayMessageSuccess _urgentquestdata)
  (CallFunction|SpectatorCompleteUrgentQuest _urgentquestdata (Variables|Default|GetQuestItemDefEventTag))
  ;; Support Quest V2
  (CallFunction|SendCustomVerbtoSquad "(TagName=\"CustomVerb.Bounty.TargetEliminated\")")
  (CallFunction|Cleanup)
  ;;[0.2135 0.2135 0.2135] This is a time out in case the player does not respond with a ServerEndQuest
  (Utilities|FlowControl|Delay 2.0)
  (CallFunction|ServerEndQuest))

(event Custom|QuestFailed
  (params (Silent :type Boolean) (Title :type Text) (Subtitle :type Text) (Event%20Tag :type Gameplay%20Tag%20Structure) (Special%20Event%20Started%20Sound :type Sound%20Base%20Soft%20Object%20Reference) (IsPoached :type Boolean))
  (bind _urgentquestdata (Utilities|Struct|MakeUrgentQuestData :EventTag Event%20Tag :VisualStyleTag "(GameplayTags=)" :TotalEventTime 6e+01 :bDisplayTimer true :bQuestCompleteOnTimeExpiration false :bQuickIntro false :bQuickOutro false :bRewardPersist false :bDisplayMultiStepExpanded true :EventTitle Title :EventSubtitle Subtitle :EventDescriptionOverrideTag "(TagName=\"\")" :SortPriority 0 :bHideQuestProgressBar false :bDisplayAllObjectivesIndividually true :QuestProgressTextFormat "LOCGEN_FORMAT_ORDERED(INVTEXT(\"`{{0}} / `{{1}}\"), INVTEXT(\"CurrentProgress\"), INVTEXT(\"MaxProgress\"))" :bReverseProgressBar true :bShowBountyThreatInformation false :bShowBountyPriceAndDistance false :bDisplayStartedEventUI true :bIsSharedUrgentQuest false :SpecialEventStartedSound Special%20Event%20Started%20Sound :AlertIconBorderColor "(R=1.000000,G=0.456411,B=0.000000,A=1.000000)" :AlertIconBrushSize "(X=0.000000,Y=0.000000)" :bOverrideWidgetPrimaryColor false :OverriddenWidgetPrimaryColor "(R=1.000000,G=0.456411,B=0.000000,A=1.000000)" :ObjectiveCategory "Default"))
  ;[0 0 0] Server_QuestFailed
  (Variables|Default|SetIsQuestCompleted false)
  (Variables|Default|SetIsPoached IsPoached)
  (if (not Silent)
    (CallFunction|DisplayMessageFailure _urgentquestdata))
  (CallFunction|SendCustomVerbtoSquad "(TagName=\"CustomVerb.Bounty.TargetPoached\")")
  (CallFunction|SpectatorFailUrgentQuest _urgentquestdata (Variables|Default|GetQuestItemDefEventTag))
  (CallFunction|Cleanup)
  ;;[0.2135 0.2135 0.2135] This is a time out in case the player does not respond with a ServerEndQuest
  (Utilities|FlowControl|Delay 2.0)
  (CallFunction|ServerEndQuest))

(event Custom|ServerEndQuest
  ;[0 0 0] Server End Quest with Success/Failure and Clean up
  (Utilities|FlowControl|DoOnce false
    (:Completed
      (if (Variables|Default|GetIsQuestCompleted)
        (CallFunction|CompleteQuest)
        (else
          (CallFunction|FailQuest)
          (if (Variables|Default|GetIsPoached)
            (CallFunction|BountyPoached)))))))

(event Custom|DisplayMessage_Failure
  (params (UrgentQuestData :type Urgent%20Quest%20Data%20Structure))
  ;[0 0 0] DisplayMessage_Failure
  (Class|AthenaHUDContext|ShowUrgentQuestEndedMessage (BlueprintContext|GetAthenaHUDContext) UrgentQuestData false)
  (CallFunction|ServerEndQuest))

(event Custom|DisplayMessage_Success
  (params (UrgentQuestData :type Urgent%20Quest%20Data%20Structure))
  ;[0 0 0] DisplayMessage_Success
  (Class|AthenaHUDContext|ShowUrgentQuestEndedMessage (BlueprintContext|GetAthenaHUDContext) UrgentQuestData true)
  (CallFunction|ServerEndQuest))

(event Custom|OnTargetDied_Bind
  (params (PawnDamageInfo :type Pawn%20Damage%20Data%20Structure%20%28by%20ref%29))
  (bind (_damagedactor _damage _instigatedby) (Utilities|Struct|BreakPawnDamageData PawnDamageInfo))
  (bind _returnvalue (Ability|GetAvatarActorfromActorInfo))
  (bind (_asfort_player_pawn_athena _bsuccess) (Utilities|Casting|CastToFortPlayerPawnAthena _returnvalue))
  (bind (_asplayer_pawn_athena _bsuccess_2) (Utilities|Casting|CastToPlayerPawn_Athena _damagedactor))
  ;[0 0 0] OnTargetDied_Bind
  (if (and _bsuccess_2 _bsuccess)
    (Variables|Default|SetVictimPawn _asplayer_pawn_athena)
    (bind _output_get (Variables|Default|SetKillerPawn (Utilities|Casting|CastToFortPlayerPawnAthena (Pawn|GetControlledPawn _instigatedby))))
    (bind _output_get_2 (Variables|Default|SetOwningPawn _asfort_player_pawn_athena))
    (if (or (Squad|OnSameSquad _output_get _output_get_2) (and (Team|OnSameTeam _output_get _output_get_2) (GameplayTags|HasMatchingGameplayTag _output_get "(TagName=\"Athena.Gameplay.Status.Converted\")")))
      (CallFunction|QuestCompleted "NSLOCTEXT(\"[38730286452D14E44F4B9F92BEE31616]\", \"99F93D3F4B81C47B0ED4CFB0706DCE02\", \"Bounty Complete!\")" "NSLOCTEXT(\"[38730286452D14E44F4B9F92BEE31616]\", \"3281CBED410391D557B976957FC2769E\", \"Target eliminated\")" "/Game/Sounds/Fort_UI_Sounds/Gameplay/Bounty/Bounty_Complete_Cue.Bounty_Complete_Cue")
      (elif (Class|AthenaFunctionLibrary|RowtoBool (Variables|Default|GetAllowPoaching))
        (CallFunction|HandlePoacher _output_get)
        (CallFunction|QuestFailed false "NSLOCTEXT(\"[38730286452D14E44F4B9F92BEE31616]\", \"F301D86843F761A9562DE1820FC8FC3E\", \"Bounty Poached\")" "NSLOCTEXT(\"[38730286452D14E44F4B9F92BEE31616]\", \"4CD77A2949962AC2725F4E803EC877EB\", \"Target eliminated\")" "(TagName=\"SpecialEvent.HuntingPlayer.IsHunter\")" "/Game/Sounds/Fort_UI_Sounds/Gameplay/Bounty/Bounty_Poached_Cue.Bounty_Poached_Cue" true)))))

(event Custom|OneMinuteLeft
  ;[0 0 0] One Minute Left Warning
  (Audio|PlaySound2D "/Game/Sounds/Fort_UI_Sounds/Gameplay/Bounty/Bounty_TimeUp_Cue.Bounty_TimeUp_Cue"))

(event Custom|SendCustomVerbToSquad
  (params (CustomVerb :type Gameplay%20Tag%20Structure))
  (bind _returnvalue (Ability|GetAvatarActorfromActorInfo))
  (Ability|HasAuthority
    (:True
      (bind _asfort_player_pawn (Utilities|Casting|CastToFortPlayerPawn _returnvalue))
      (bind _returnvalue_2 (Fortnite|Utilities|AttemptGetPlayerStateforPawn _asfort_player_pawn))
      (bind _asfort_player_state_athena (Utilities|Casting|CastToFortPlayerStateAthena _returnvalue_2))
      (bind _returnvalue_3 (Class|FortGameStateAthena|GetSquadMemberPlayerStates (Fortnite|Athena|GameState|GetGameStateAthenafromContext _returnvalue) _asfort_player_state_athena true))
      (for _array_element _returnvalue_3
        (CustomVerb|BroadcastCustomVerbMessage (CustomVerb|MakeCustomVerbMessage _array_element CustomVerb))))))

(event Custom|Cleanup
  ;[0 0 0] Cleanup
  (CallFunction|RemoveTrackingCircle)
  (CallFunction|CleanupBountyHunter))

(event Ability|EventOnEndAbility
  (params (bWasCancelled :type Boolean))
  ;[0 0 0] OnEndAbility
  (CallFunction|Cleanup))

(event Ability|EventFailedToActivatePassiveAbility
  (params (ActorInfo :type Gameplay%20Ability%20Actor%20Info%20Structure)))

(event Custom|QuestStarted)

(event Custom|OnHunterDied
  (params (DamagedActor :type Actor%20Object%20Reference) (Damage :type Float%20%28single-precision%29) (InstigatedBy :type Controller%20Object%20Reference) (DamageCauser :type Actor%20Object%20Reference) (HitLocation :type Vector) (FHitComponent :type Primitive%20Component%20Object%20Reference) (BoneName :type Name) (Momentum :type Vector))
  (CallFunction|QuestFailed true "NSLOCTEXT(\"[38730286452D14E44F4B9F92BEE31616]\", \"F301D86843F761A9562DE1820FC8FC3E\", \"Bounty Poached\")" "NSLOCTEXT(\"[38730286452D14E44F4B9F92BEE31616]\", \"4CD77A2949962AC2725F4E803EC877EB\", \"Target eliminated\")" "(TagName=\"SpecialEvent.HuntingPlayer.IsHunter\")" "/Game/Sounds/Fort_UI_Sounds/Gameplay/Bounty/Bounty_Poached_Cue.Bounty_Poached_Cue" true))

(fn AddTrackingCircleInfo
  (params (Max%20Center%20Offset :type Float%20%28double-precision%29 :default 3.5e+04) (Minimum%20Radius :type Float%20%28double-precision%29 :default 1e+03) (Maximum%20Radius :type Float%20%28double-precision%29 :default 5e+04) (Time%20Between%20Updates%20in%20Seconds :type Float%20%28double-precision%29 :default 5.0) (Radius%20Amount%20to%20Shrink%20Each%20Update :type Float%20%28double-precision%29) (Circle%20Color :type Linear%20Color%20Structure))
  (bind (_owneractor _avataractor _playercontroller) (Utilities|Struct|BreakGameplayAbilityActorInfo (Ability|GetActorInfo)))
  (Utilities|Casting|CastToFortPlayerController _playercontroller
    (:then
      (bind _bountytarget (CallFunction|GetBountyTarget))
      (bind _outpawn (Class|AbilityFunctions|GetPawnfromPlayerStateZone _bountytarget)) ; Get the pawn from the player state
      (Variables|Default|SetTracked%20Hunter%20Bounty%20Target%20Pawn _outpawn) ; Store the pawn to remove later, just in case the player respawns\nbefore the pawn is removed (so we don't get the wrong pawn to remove).
      (bind _trackingcircleactors (CallFunction|GetTrackingCircleActors))
      (bind _returnvalue (RadiusTracker|MakeRadiusTrackingInitializer Max%20Center%20Offset Minimum%20Radius Maximum%20Radius Time%20Between%20Updates%20in%20Seconds Radius%20Amount%20to%20Shrink%20Each%20Update Circle%20Color 0 _outpawn _trackingcircleactors))
      (bind _returnvalue_2 (Fortnite|AddActorTrackingInfo (RadiusTracker|GetGameStateRadiusTrackerComponent) _returnvalue))
      (Variables|Default|SetTrackingKey _returnvalue_2))
    (:CastFailed
      (Development|PrintString "Ability Owner Player Controller is not a valid FortPlayerController.  Cannot add tracking info for radius display." false)))
  (return))

(fn GetBountyTarget
  (return (CallFunction|GetTrackedHunterBountyTarget)))

(fn GetTrackingCircleActors
  (bind (_owneractor _avataractor _playercontroller) (Utilities|Struct|BreakGameplayAbilityActorInfo (Ability|GetActorInfo)))
  (bind _asfort_player_controller (Utilities|Casting|CastToFortPlayerController _playercontroller)
    (:then
      (return (Utilities|Array|MakeArray _asfort_player_controller)))
    (:CastFailed
      (return 0))))

(fn HandlePoacher
  (params (Killer :type Fort%20Player%20Pawn%20Object%20Reference))
  (bind _returnvalue (Fortnite|GetFortPlayerControllerfromActor Killer))
  (bind _returnvalue_2 (Quests|GetQuestManager _returnvalue "Athena"))
  (Utilities|IsValid Killer
    (:Is%20Valid
      (Utilities|IsValid _returnvalue
        (:Is%20Valid
          (Utilities|IsValid _returnvalue_2
            (:Is%20Valid
              (Quest|SendComplexCustomStatEvent _returnvalue_2 0 "(GameplayTags=)" "(GameplayTags=((TagName=\"Athena.Quests.Bounty.Poacher\")))" 1))))))))

(fn RemoveTrackingCircle
  (bind _returnvalue (RadiusTracker|GetGameStateRadiusTrackerComponent))
  (Utilities|IsValid _returnvalue
    (:Is%20Valid
      (RadiusTracker|RemoveTrackingforActorbyTrackingKey _returnvalue (Variables|Default|GetTracked%20Hunter%20Bounty%20Target%20Pawn) (Variables|Default|GetTrackingKey)))
    (:Is%20Not%20Valid
      (Development|PrintString "GameStateRadiusTrackerComponent is NOT Valid"))))
